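# Translucent unlit Niagara ribbon material for Nexus environment interaction effects, drawing flowing, glowing energy streaks.
# Two panning magnetic-wave samples and blended caustics add to a sphere mask, tinted by a two-color gradient and particle color, then HDR-boosted for emissive.
# Opacity combines panning blurred noise with edge and end fades, depth fade and particle alpha.
# Only the Random dynamic parameter is used, offsetting pan time to vary ribbons.
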Material /FNGameplayCues/Athena/Environments/Nexus/Materials/M_ZP_Interacttion_Ribbons
graph {
 properties {bUsedWithNiagaraRibbons=true, blendMode=BLEND_Translucent, pixelDepthOffsetMode=PDOM_Legacy, shadingModel=MSM_Unlit, usedShadingModels=Unlit}
 nodes [
  n1 = ComponentMask(None=TextureCoordinate()) {r=true}
  n2 = VectorParameter() {parameterName=Tile1, defaultValue=<rgba 0.5 0.125 0 0>}
  n3 = Multiply(A=TextureCoordinate(), B=AppendVector(A=@n2.R, B=@n2.R))
  n4 = VectorParameter() {parameterName=PanDirection1, defaultValue=<rgba 0.2 0.02 0 0>}
  n5 = AppendVector(A=@n4.R, B=@n4.R)
  n6 = Add(
   A=Multiply(A=Time(), B=ScalarParameter() {defaultValue=-5, parameterName=PanSpeed})
   B=DynamicParameter() {defaultValue=<rgba 1 1 12 0>, paramNames [Random,Erode,CenterPower,Squeeze]}.Random
  )
  n7 = ScalarParameter() {defaultValue=6.283, parameterName=Rotation}
  n8 = Rotator(Coordinate=Panner(Coordinate=Multiply(A=@n3) {constB=1.5}, Time=@n6, Speed=@n5) {speedX=-0.2, speedY=-0.02}, Time=@n7)
  n9 = VectorParameter() {parameterName=PanDirection1Mult, defaultValue=<rgba 1 -1 0 1>}
  n10 = AppendVector(A=@n9.R, B=@n9.R)
  n11 = Rotator(Coordinate=Panner(Coordinate=@n3, Time=@n6, Speed=Multiply(A=@n5, B=@n10)) {speedX=-0.2, speedY=0.02}, Time=@n7)
  n12 = VectorParameter() {parameterName=Tile2, defaultValue=<rgba 0.65 0.25 0 0>}
  n13 = VectorParameter() {parameterName=PanDirection2, defaultValue=<rgba 0.1 0.04 0 0>}
  n14 = Panner(Coordinate=Multiply(A=TextureCoordinate(), B=AppendVector(A=@n12.R, B=@n12.R)), Time=@n6, Speed=AppendVector(A=@n13.R, B=@n13.R)) {speedX=0.1, speedY=0.04}
  n15 = ScalarParameter() {defaultValue=6.283, parameterName=Rotation}
  n16 = Rotator(Coordinate=Add(A=@n14) {constB=0.01}, Time=@n15)
  n17 = Rotator(Coordinate=@n14, Time=@n15)
  n18 = Add(
   A=Add(
    A=TextureSample(UVs=@n8) {texture=/Game/Athena/Prototype/Blueprints/Sneezy/Textures/T_Sneezy_MagneticWaves}.RGB
    B=TextureSample(UVs=@n11) {texture=&T_Sneezy_MagneticWaves}.RGB
   )
   B=LinearInterpolate(
    A=TextureSample(UVs=@n16) {texture=/Game/Global/Textures/Default/Caustics/T_Caustics_Softer}.G
    B=TextureSample(UVs=@n17) {texture=&T_Caustics_Softer}.G
   )
  )
  n19 = Power(Base=SphereMask(A=TextureCoordinate() {uTiling=0.65}, B=Constant2Vector() {g=0.5}, Radius=Constant() {r=0.5}) {hardnessPercent=0}) {constExponent=1.5}
  n20 = Multiply(
   A=LinearInterpolate(
    A=VectorParameter() {parameterName=Colo_01, defaultValue=<rgba 0.7608 0 1 1>}.RGB
    B=VectorParameter() {parameterName=Colo_02, defaultValue=<rgba 0 0.3622 1 1>}.RGB
    Alpha=Power(Base=@n1) {constExponent=0.325}
   )
   B=Multiply(A=ParticleColor().RGB, B=Add(A=@n18, B=@n19))
  )
  MaterialFunctionCall(Emissive=MaterialFunctionCall(LDR Input=@n20, HDR Tint=@n20, HDR Intensity Multiplier=ScalarParameter() {defaultValue=9, parameterName=HDR Intensity}) {materialFunction=&DeriveHDRfromLDR}.RemappedRange) {materialFunction=&MF_FX_EmissiveAdjustments}
  n21 = Multiply(A=TextureCoordinate(), B=ComponentMask(None=VectorParameter() {defaultValue=<rgba 1 1 0 1>}.RGB) {g=true, r=true})
  n22 = TextureSample(UVs=Rotator(Coordinate=Panner(Coordinate=@n21) {speedX=0.2}, Time=ScalarParameter() {defaultValue=6.283, parameterName=Rotation})) {samplerType=SAMPLERTYPE_Grayscale, texture=&LowResBlurredNoise}
  n23 = ComponentMask(None=TextureCoordinate()) {g=true}
  n24 = Multiply(A=OneMinus(None=@n23), B=@n23)
  n25 = Multiply(A=@n24, B=@n24)
  n26 = Clamp(None=Multiply(A=Multiply(A=@n22.G, B=@n25)) {constB=15})
  n27 = ComponentMask(None=TextureCoordinate()) {r=true}
  n28 = Multiply(A=Power(Base=OneMinus(None=@n27)) {constExponent=1.5}, B=Saturate(None=Multiply(A=@n27) {constB=5}))
  Multiply(A=ParticleColor().A, B=DepthFade(Opacity=Multiply(A=@n26, B=@n28)) {fadeDistanceDefault=250})
 ]
}
